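; Battle pass quest ability that credits the player for honking a vehicle horn in valid POIs.
; A driving gameplay event binds the vehicle's horn event, while a vehicle exit event ends the ability.
; On honk, it checks the pawn's required and POI tags, sends the quest stat, then ends.
; An IsOpponentInRange helper scans for nearby enemy pawns but is not called here, despite debug strings naming Honk_Near_Enemy.
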(event Custom|On%20Vehicle%20Horn
  (params (Controller :type Fort%20Player%20Controller%20Athena%20Object%20Reference) (Vehicle :type Fort%20Athena%20Vehicle%20Object%20Reference) (bIsHeld :type Boolean))
  (bind _returnvalue (MissionHelpers|HasTags (GameplayTags|GetOwnedGameplayTags (Variables|GAVars|Data|GetOwningPlayerPawn) :!class /Script/GameplayTags.GameplayTagAssetInterface) (Variables|Default|GetRequiredTags) true false))
  (if _returnvalue
    (if (GameplayTags|HasAnyTags (GameplayTags|GetOwnedGameplayTags (Variables|GAVars|Data|GetOwningPlayerPawn) :!class /Script/GameplayTags.GameplayTagAssetInterface) (Variables|Default|GetT_ValidPOIs))
      (Quest|SendComplexCustomStatEvent (Variables|GAVars|Data|GetOwningPlayerQuestManager) 0 (Variables|Default|GetQuestTag) "(GameplayTags=)" 1)
      (Ability|EndAbility))))

(event Ability|EventActivateAbilityFromEvent
  (params (EventData :type Gameplay%20Event%20Data%20Structure%20%28by%20ref%29))
  (|CallParentFunction EventData)
  (switch Utilities|FlowControl|Switch|SwitchonGameplayTag (Utilities|Struct|BreakGameplayEventData EventData)
    (:Event.Vehicle.Exit
      (Development|PrintString "@@@ ActivateAbilityFromEvent [GA_Quest_Honk_Near_Enemy]: CALLED; Vehicle Exited!" false (Variables|Debug|GetbDebugAbility))
      (Ability|EndAbility))
    (:MovementMode.Driving
      (Development|PrintString "@@@ ActivateAbilityFromEvent [GA_Quest_Honk_Near_Enemy]: CALLED; Driving vehicle!" false (Variables|Debug|GetbDebugAbility))
      (bind _returnvalue (Vehicle|GetVehicleActor (Vehicle|GetVehicleInterface (Variables|GAVars|Data|GetOwningPlayerPawn)) :!class /Script/FortniteGame.FortVehicleInterface))
      (bind _asfort_athena_vehicle (Utilities|Casting|CastToFortAthenaVehicle _returnvalue)
        (:then
          (bind _output_get (Variables|Default|SetFortAthenaVehicle _asfort_athena_vehicle))
          (Vehicle|BindEventtoOnVehicleHornEvent _output_get (delegate Custom|On%20Vehicle%20Horn)))
        (:CastFailed
          (Ability|EndAbility))))))

(fn IsOpponentInRange
  (params (TargetVehicle :type Fort%20Athena%20Vehicle%20Object%20Reference))
  (LocalVariables|SetbFoundEnemyInRange false)
  (bind _outfortpawns (Fortnite|GetFortPawnsInRange (Transformation|GetActorLocation TargetVehicle) (* (LocalVariables|GetradiusInMeters) 1e+02)))
  (bind _array_element (Utilities|Array|ForEachLoopwithBreak _outfortpawns)
    (:LoopBody
      (bind _asfort_player_pawn_athena (Utilities|Casting|CastToFortPlayerPawnAthena _array_element))
      (if (not (Team|OnSameTeam _asfort_player_pawn_athena (Variables|GAVars|Data|GetOwningPlayerPawn)))
        (LocalVariables|SetbFoundEnemyInRange true)))
    (:Completed
      _))
  (return (LocalVariables|GetbFoundEnemyInRange)))
